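# Physics asset for the Gone Fishing mic stand cosmetic, giving its fishing props simulated bodies built from capsules, plus spheres for the two string bodies.
# The handle is the hub, with can, bottle, lures, strings, rod, bucket and net base constrained to it.
# The net and fish tail form three-body chains branching from the net base.
# Twist is locked on most joints, and the rod and bucket also lock one swing axis to act as hinges.
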PhysicsAsset /SparksCosmetics/Instrument/MicStand/GoneFishing/Meshes/GoneFishing_MicStand_Physics
bodies {
 bottle [{shapeType=Capsule, length=7.4312, radius=2.7202, shapeName=bottle_capsule, center=<0 0 5>, rotation=<rot 5.0916 -90 3.1368>}]
 bucket_hinge [{
   shapeType=Capsule
   length=12.8184
   radius=6.2815
   shapeName=bucket_hinge_capsule
   center=<-6.7991 0.1108 0>
   rotation=<rot 0 -0.9337 -89.9949>
  }]
 can [{shapeType=Capsule, length=3.5354, radius=2.9005, shapeName=can_capsule, center=<0 0 5>, rotation=<rot -3.3214 -90 4.8119>}]
 fish_tail_01 [{shapeType=Capsule, length=2.0798, radius=2.1876, shapeName=fish_tail_01_capsule, rotation=<rot 90 0 0>}]
 fish_tail_02 [{shapeType=Capsule, length=3.7666, radius=2.4503, shapeName=fish_tail_02_capsule, center=<2 0 0.5>, rotation=<rot -3.044 -90 -4.9888>}]
 fish_tail_end [{shapeType=Capsule, length=6.1683, radius=1.8518, shapeName=fish_tail_end_capsule, center=<1.3 0 0>, rotation=<rot -13.7467 -90 -5.6249>}]
 handle [{shapeType=Capsule, length=59.6871, radius=1.4734, center=<-0.4742 0 15.2247>}]
 lure_01 [{shapeType=Capsule, length=4.1623, radius=2, shapeName=lure_01_capsule, center=<5 0 1>, rotation=<rot 0 90.4567 -82.2178>}]
 lure_02 [{shapeType=Capsule, length=4.5008, radius=2.5, shapeName=lure_02_capsule, center=<6 0 1.2>, rotation=<rot 0 90.0702 -81.8494>}]
 lure_03 [{shapeType=Capsule, length=5, radius=2.095, shapeName=lure_03_capsule, center=<7 0 0.8>, rotation=<rot 0 89.991 -84.4527>}]
 net_01 [{shapeType=Capsule, length=5.6341, radius=4.7844, shapeName=net_01_capsule, center=<2 -0.5 1>, rotation=<rot 0.0001 0 58.7717>}]
 net_02 [{shapeType=Capsule, length=4.4741, radius=4.5178, shapeName=net_02_capsule, center=<0.8 0 0>, rotation=<rot 0.0001 0 58.7717>}]
 net_03 [{shapeType=Capsule, length=2.2279, radius=3.0665, shapeName=net_03_capsule, center=<0.7 -0.5 0.7>, rotation=<rot 0.0001 0 58.7717>}]
 net_base [
  {shapeType=Capsule, length=14, radius=0.9698, shapeName=net_base_capsule, center=<7 0 0>, rotation=<rot 0 -90 -90>}
  {shapeType=Capsule, length=10.8725, radius=1.2605, shapeName=net_base_capsule_l, center=<17.4954 2.9973 -1.325>, rotation=<rot 78 30 -3>}
  {
   shapeType=Capsule
   length=10.8138
   radius=1.3192
   shapeName=net_base_capsule_r
   center=<18.1059 -3.9618 0.9525>
   rotation=<rot 79.8665 144.4645 -179.9998>
  }
 ]
 rod_rotate [{shapeType=Capsule, length=6.7937, radius=2, shapeName=rod_rotate_capsule, center=<3.2 -6 0>, rotation=<rot 0 0 90>}]
 string_bottom [{shapeType=Sphere, radius=4, shapeName=string_bottom_sphere, center=<6 0 0>}]
 string_top [{shapeType=Sphere, radius=4, shapeName=string_top_sphere, center=<5 -0.6 -2>}]
}
constraints [
 {bone1Name=can, bone2Name=handle, swing1LimitDegrees=2, swing2LimitDegrees=2, twistLimitDegrees=45, twistMotion=Locked}
 {bone1Name=bottle, bone2Name=handle, swing1LimitDegrees=2, swing2LimitDegrees=2, twistLimitDegrees=45, twistMotion=Locked}
 {bone1Name=lure_03, bone2Name=handle, swing1LimitDegrees=45, swing2LimitDegrees=90, twistLimitDegrees=45, twistMotion=Locked}
 {bone1Name=lure_01, bone2Name=handle, swing1LimitDegrees=45, swing2LimitDegrees=90, twistLimitDegrees=45, twistMotion=Locked}
 {bone1Name=lure_02, bone2Name=handle, swing1LimitDegrees=45, swing2LimitDegrees=90, twistLimitDegrees=45, twistMotion=Locked}
 {bone1Name=net_base, bone2Name=handle, swing1LimitDegrees=4, swing2LimitDegrees=4, twistLimitDegrees=45, twistMotion=Locked}
 {bone1Name=net_01, bone2Name=net_base, swing1LimitDegrees=30, swing2LimitDegrees=40, twistLimitDegrees=3}
 {bone1Name=net_02, bone2Name=net_01, swing1LimitDegrees=50, swing2LimitDegrees=60, twistLimitDegrees=3}
 {bone1Name=net_03, bone2Name=net_02, swing1LimitDegrees=40, swing2LimitDegrees=35, twistLimitDegrees=3}
 {bone1Name=fish_tail_end, bone2Name=fish_tail_02, swing1LimitDegrees=5, swing2LimitDegrees=10, twistLimitDegrees=8, twistMotion=Locked}
 {bone1Name=fish_tail_01, bone2Name=net_base, swing1LimitDegrees=10, swing2LimitDegrees=20, twistLimitDegrees=2, twistMotion=Locked}
 {bone1Name=fish_tail_02, bone2Name=fish_tail_01, swing1LimitDegrees=35, swing2LimitDegrees=30, twistLimitDegrees=20}
 {bone1Name=string_top, bone2Name=handle, swing1LimitDegrees=80, swing2LimitDegrees=3, twistLimitDegrees=45, twistMotion=Locked}
 {bone1Name=string_bottom, bone2Name=handle, swing1LimitDegrees=75, swing2LimitDegrees=3, twistLimitDegrees=45, twistMotion=Locked}
 {
  bone1Name=rod_rotate
  bone2Name=handle
  swing1LimitDegrees=45
  swing1Motion=Locked
  swing2LimitDegrees=180
  twistLimitDegrees=45
  twistMotion=Locked
 }
 {
  bone1Name=bucket_hinge
  bone2Name=handle
  swing1LimitDegrees=45
  swing1Motion=Locked
  swing2LimitDegrees=22
  twistLimitDegrees=45
  twistMotion=Locked
 }
]
