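# Animation Blueprint for the Gone Fishing mic stand cosmetic, adding physics and looping wiggle to its net and fish tail.
# A RigidBody node simulates relative to the handle bone, then additive bone transforms rotate net_01 and fish_tail_01.
# The update event drives those rotations with real-time sine waves and switches the physics gravity override for normal, skydiving and parachute states.
# The fish_tail_02 transform is unconnected, and wave_04 is computed but unused.
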AnimBlueprint /SparksCosmetics/Instrument/MicStand/GoneFishing/Meshes/GoneFishing_MicStand_AnimBP
target_skeleton=&/GoneFishing_MicStand_Skeleton
graphs [
 AnimGraph {nodes [
   "Transform (Modify) Bone - Bone: fish_tail_02"(Alpha=1, Translation="0,0,0", Rotation="0,0,90", Scale="1,1,1") {
    rotationMode=BMM_Additive
    rotationSpace=BCS_BoneSpace
    alphaBoolBlend {blendOption=Linear}
    boneToModify {boneName=fish_tail_02}
    rotation {roll=90}
   }
   n1 = RigidBody(ComponentPose=Local To Component(LocalPose=Input Pose()), ExternalForce=Get GravityOverride()) {
    bOverrideWorldGravity=true
    overridePhysicsAsset=&/GoneFishing_MicStand_Physics
    simulationSpace=BaseBoneSpace
    alphaBoolBlend {blendOption=Linear}
    baseBoneRef {boneName=handle}
    componentAppliedLinearAccClamp=<3000 3000 3000>
    componentLinearAccScale=<0.4 0.4 0.4>
    componentLinearVelScale=<0.2 0.2 0.2>
    overrideWorldGravity {z=-980}
    simSpaceSettings {
     maxAngularAcceleration=1000
     maxAngularVelocity=6
     maxLinearAcceleration=700
     maxLinearVelocity=600
     velocityScaleZ=0.1
     worldAlpha=0.5
     externalLinearDragV=<0.2 0.2 0.02>
    }
   }
   n2 = "Transform (Modify) Bone - Bone: net_01"!(
    ComponentPose=@n1
    bAlphaBoolEnabled:bool=True
    Translation="0,0,0"
    Rotation_Roll:float=Get wave_01() -> (wave_01:double)
    Rotation_Pitch:float=0
    Rotation_Yaw:float=Get wave_03() -> (wave_03:double)
    Scale="1,1,1"
   ) -> (Pose) {
    alphaInputType=Bool
    rotationMode=BMM_Additive
    rotationSpace=BCS_ParentBoneSpace
    alphaBoolBlend {blendOption=Linear}
    boneToModify {boneName=net_01}
   }
   n3 = Get Tail() -> (Tail:double)
   n4 = "Transform (Modify) Bone - Bone: fish_tail_01"!(
    ComponentPose=@n2
    bAlphaBoolEnabled:bool=True
    Translation="0,0,0"
    Rotation_Roll:float=@n3
    Rotation_Pitch:float=@n3
    Rotation_Yaw:float=0
    Scale="1,1,1"
   ) -> (Pose) {
    alphaInputType=Bool
    rotationMode=BMM_Additive
    rotationSpace=BCS_ParentBoneSpace
    alphaBoolBlend {blendOption=Linear}
    boneToModify {boneName=fish_tail_01}
   }
   Output Pose(Result=Component To Local(ComponentPose=@n4))
  ]}
 EventGraph {nodes [
   Event BlueprintUpdateAnimation()
   > n1 = Sequence()
   > n2 = Branch(Condition=Get bIsSkyDiving())
   > n3 = Branch(Condition=Get bIsParachuteOpened())
   > Set GravityOverride(GravityOverride="0,0,-100")
   Set GravityOverride(execute=@n3.else, GravityOverride="0,0,20000")
   Set GravityOverride(execute=@n2.else, GravityOverride="0,0,-980")
   n4 = GetRealTimeSeconds()
   n5 = float * float(A=@n4, B=6)
   Set wave_01(execute=@n1.then_1, wave_01:double=float * float(A="Sin (Radians)"(A=@n5), B=2)) -> (Output_Get:double)
   n6 = float * float(A=@n4, B=8)
   > Set Tail(Tail:double=float * float(A="Sin (Radians)"(A=@n6), B=4)) -> (Output_Get:double)
   n7 = float * float(A=@n4, B=4)
   > Set wave_03(wave_03:double=float * float(A="Sin (Radians)"(A=@n7), B=8)) -> (Output_Get:double)
   n8 = float * float(A=@n4, B=4)
   > Set wave_04(wave_04:double=float * float(A="Sin (Radians)"(A=@n8), B=2)) -> (Output_Get:double)
   TryGetPawnOwner()
  ]}
]
variables [
 GravityOverride {type=Vector}
 Tail {type=number}
 wave_01 {type=number}
 wave_03 {type=number}
 wave_04 {type=number}
]
